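# Sound cue for the Season 23 battle pass Part 2 Quest 02 signal tracking lock-confirm effect, routed to the Fort_Quests sound class.
# A Random node picks one of three equally weighted LockConfirm wave variations.
# A Modulator at the root adds a very slight downward pitch variation while keeping volume constant.
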SoundCue /BattlePassS23/Audio/SFX/S23_Pt02_Q02_SignalTrack_LockConfirm_Cue
root=Modulator_0
nodes [
 {id=Modulator_0, children [Random_0], properties {pitchMax=1, pitchMin=0.99, volumeMax=1, volumeMin=1}}
 {id=Random_0, children [WavePlayer_3,WavePlayer_4,WavePlayer_5], properties {weights [1,1,1]}}
 {id=WavePlayer_3, properties {soundWaveAssetPtr=&/S23_Pt02_Q02_SignalTrack_LockConfirm_01}}
 {id=WavePlayer_4, properties {soundWaveAssetPtr=&/S23_Pt02_Q02_SignalTrack_LockConfirm_02}}
 {id=WavePlayer_5, properties {soundWaveAssetPtr=&/S23_Pt02_Q02_SignalTrack_LockConfirm_03}}
]
properties {maxDistance=2097152, soundClassObject=&Fort_Quests, volumeMultiplier=1}
