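# Sound cue for the success sting of a Season 23 Battle Pass signal-tracking quest, routed to the Fort_Quests sound class.
# It picks one of three success waves at random with equal weight.
# A modulator at the root adds a very slight downward pitch variation and leaves volume unchanged.
# The cue carries a very large max distance.
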SoundCue /BattlePassS23/Audio/SFX/S23_Pt02_Q02_SignalTrack_Success_Cue
root=Modulator_0
nodes [
 {id=Modulator_0, children [Random_0], properties {pitchMax=1, pitchMin=0.99, volumeMax=1, volumeMin=1}}
 {id=Random_0, children [WavePlayer_0,WavePlayer_1,WavePlayer_2], properties {weights [1,1,1]}}
 {id=WavePlayer_0, properties {soundWaveAssetPtr=&/S23_Pt02_Q02_SignalTrack_Success_01}}
 {id=WavePlayer_1, properties {soundWaveAssetPtr=&/S23_Pt02_Q02_SignalTrack_Success_02}}
 {id=WavePlayer_2, properties {soundWaveAssetPtr=&/S23_Pt02_Q02_SignalTrack_Success_03}}
]
properties {maxDistance=2097152, soundClassObject=&Fort_Quests}
